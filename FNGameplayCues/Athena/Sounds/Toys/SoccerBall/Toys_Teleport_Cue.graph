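# Sound cue that plays the teleport sound for Fortnite toys, stored under the Athena SoccerBall toy gameplay cue sounds.
# A single Toys_Teleport wave player feeds a Modulator at the root.
# It plays in the gadgets sound class and SFX submix at reduced volume, with Creative small-toy attenuation limiting how far it carries.
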SoundCue /FNGameplayCues/Athena/Sounds/Toys/SoccerBall/Toys_Teleport_Cue
root=Modulator_0
nodes [
 {id=Modulator_0, children [WavePlayer_0]}
 {id=WavePlayer_0, properties {soundWaveAssetPtr=&/Toys_Teleport}}
]
properties {
 attenuationSettings=/Game/Sounds/Creative/Attenuations/Fort_Creative_SmallToy_Att
 maxDistance=3000
 soundClassObject=&Fort_Gadgets
 soundSubmixObject=&Fortnite_SFX
 volumeMultiplier=0.6696
}
